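; BP_DonutCloud is a donut cloud actor that can fade in and out and broadcast when it has finished forming or fading.
; Fade In plays the EnableCloud timeline from the start and Fade Out reverses it, driving the WPOCollapse scalar in the MPC_WildEstate_TA material parameter collection.
; The timeline's end direction selects the OnCloudFinishedForming or OnCloudFinishedFading dispatcher.
; BeginPlay waits one frame, then enables custom depth on the gas cloud mesh.
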(event EventBeginPlay
  (Utilities|FlowControl|DelayUntilNextFrame)
  (Rendering|SetRenderCustomDepth (Variables|Default|GetGasCloud_A2_StaticMeshComponent0) true))

(event Custom|Fade%20In%20Cloud
  (bind (_direction _enablefloat) (Timeline|EnableCloud PlayFromStart)
    (:Update
      (Rendering|Material|SetScalarParameterValue "/WildEstate/TechArt/MPCs/MPC_WildEstate_TA.MPC_WildEstate_TA" "WPOCollapse" _enablefloat :!class /Script/Engine.KismetMaterialLibrary))
    (:Finished
      (switch Utilities|FlowControl|Switch|SwitchonETimelineDirection _direction
        (:Forward
          (Default|CallOnCloudFinishedForming))
        (:Backward
          (Default|CallOnCloudFinishedFading))))))

(event Custom|Fade%20Out%20Cloud
  (Timeline|EnableCloud Reverse))

(fn ConstructionScript)
