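# Masked unlit material for a stylized, glowing donut cloud in WildEstate, with an IsBody static switch separating the inset body from its puff cards.
# Vertex colors drive card randomness, thickness and face masking, while WPO scales and offsets camera-facing sprites with sine wobble and a global collapse parameter.
# Color combines stepped sunlight shading, time-of-day tints, pulsing interior glow and fresnel.
# Rotating puff textures and noise feed outlines, height-lerp camera fade, dissolve and pixel depth offset.
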Material /WildEstate/TechArt/DonutCloud/Materials/M_DonutCloud
graph {
 properties {
  bAutomaticallySetUsageInEditor=false
  bHasPixelAnimation=true
  bUsedWithMeshDeformer=true
  bUsedWithSkeletalMesh=true
  blendMode=BLEND_Masked
  shadingModel=MSM_Unlit
  usedShadingModels=Unlit
 }
 nodes [
  n1 = WorldPosition()
  n2 = MaterialFunctionCall(Color=@n1.XYZ) {materialFunction=&SplitComponents}
  n3 = ComponentMask(None=@n1.XYZ) {b=true}
  n4 = VectorParameter() {parameterName=SphereLocation, defaultValue=<rgba 0 0 0 1>}
  n5 = MaterialFunctionCall(R=@n2.R, G=@n2.R, B=If(A=@n3, B=@n4.B, "A > B"=@n3, "A < B"=@n4.B)) {materialFunction=&AppendMany}
  n6 = ScalarParameter() {defaultValue=93.0634, group=InteriorGLow, parameterName=InteriorGlow_SphereRadius}
  n7 = ScalarParameter() {defaultValue=1, group=InteriorGLow, parameterName=InteriorGlow_SphereHardness}
  Abs(None=OneMinus(None=SphereMask(A=@n5.RGB, B=@n4.RGB, Radius=@n6, Hardness=@n7)))
  n8 = CameraPositionWS()
  n9 = ActorPositionWS()
  n10 = DotProduct(A=Normalize(VectorInput=Subtract(A=@n8, B=@n9)), B=Constant3Vector() {constant=<rgba 0 0 -1 0>})
  n11 = LinearInterpolate(Alpha=SmoothStep(
    Min=ScalarParameter() {defaultValue=0.95, parameterName=AngleMask.Min}
    Max=ScalarParameter() {defaultValue=1, parameterName=AngleMask.Max}
    Value=ConstantBiasScale(None=@n10)
   ) {constMin=0.97}) {constA=1, constB=0}
  n12 = ComponentMask(None=TextureCoordinate()) {g=true}
  Add(A=@n11, B=Subtract(A=OneMinus(None=@n12)))
  Add(A=ObjectPositionWS(), B=Constant3Vector() {constant=<rgba 0 0 5000 0>})
  n13 = ActorPositionWS()
  n14 = MaterialFunctionCall(Vector 2=ComponentMask(None=@n13) {b=true, r=true}) {materialFunction=/Engine/Functions/Engine_MaterialFunctions02/Utility/VectorLength}
  n15 = ComponentMask(None=@n13) {g=true}
  n16 = Subtract(A=Abs(None=AppendVector(A=@n14.V2 Length, B=@n15)), B=ScalarParameter() {defaultValue=512, group=Cylindarmask, parameterName=CylMaskHeight})
  n17 = ComponentMask(None=@n16) {r=true}
  n18 = ComponentMask(None=@n16) {g=true}
  Add(A=Distance(A=Max(A=@n16) {constB=0}, B=Constant()), B=Min(A=Max(A=@n17, B=@n18) {constB=0}) {constB=0})
  n19 = WorldPosition()
  ComponentMask(None=@n19.XYZ) {g=true, r=true}
  n20 = ObjectPositionWS()
  ComponentMask(None=Subtract(
    A=TransformPosition(None=@n19.XYZ) {transformSourceType=TRANSFORMPOSSOURCE_World}
    B=TransformPosition(None=@n20) {transformSourceType=TRANSFORMPOSSOURCE_World}
   )) {g=true, r=true}
  ComponentMask(None=NamedRerouteUsage()) {r=true}
  Constant()
  n21 = LinearInterpolate(A=NamedRerouteUsage(), B=VectorParameter() {group=Color, parameterName=ColorOutline, defaultValue=<rgba 0.2346 0.1369 0.5417 1>}.RGB) {constAlpha=0.85}
  n22 = VectorParameter() {group=Color, parameterName=Color2, defaultValue=<rgba 0.0835 0.0455 0.2031 1>}
  n23 = CollectionParameter() {
   collection=&FortniteMaterialParameters
   parameterName=SunlightDirectionalVector
  }
  n24 = NamedRerouteUsage()
  n25 = ComponentMask(None=VertexNormalWS()) {b=true}
  n26 = Add(B=@n25) {constA=-0.1}
  n27 = Max(A=DotProduct(A=NamedRerouteUsage(), B=@n23), B=Add(A=Multiply(A=@n24) {constB=0.05}, B=@n26))
  n28 = Max(A=@n27, B=@n26)
  n29 = ScalarParameter() {defaultValue=1.85, group=Color, parameterName=ShadinOffset}
  n30 = Add(
   A=Add(A=NamedRerouteUsage(), B=StaticSwitchParameter(True=@n27, False=@n28) {group=00_Switches, parameterName=IsBody, sortPriority=0})
   B=@n29
  )
  n31 = Multiply(A=@n24) {constB=0.3}
  n32 = Add(A=@n25) {constB=0.7}
  n33 = ComponentMask(None=WorldPosition().XYZ) {b=true}
  n34 = ComponentMask(None=MaterialFunctionCall() {materialFunction=&ObjectPivotPoint}.Object Pivot Location) {b=true}
  n35 = MaterialFunctionCall() {materialFunction=&ObjectLocalBounds}
  n36 = Saturate(None=Divide(A=Subtract(A=@n33, B=@n34), B=ComponentMask(None=@n35.Local Bounds Size) {b=true}))
  n37 = Power(Base=Add(
    A=StaticSwitchParameter(True=@n27, False=@n36) {group=00_Switches, parameterName=IsBody, sortPriority=0}
    B=ScalarParameter() {defaultValue=1.75, group=Color, parameterName=ShadingOffsetVertexNormal}
   ) {constB=-0.5}) {constExponent=8}
  n38 = Multiply(A=NamedRerouteUsage(), B=ScalarParameter() {defaultValue=2, group=Color, parameterName=ShadingOffsetNoiseAmount}) {constB=2}
  n39 = Min(A=@n37, B=Add(A=@n37, B=Saturate(None=@n38)))
  n40 = Min(
   A=Saturate(None=Divide(A=@n30))
   B=Min(A=Add(A=@n31, B=@n32), B=StaticSwitchParameter(True=@n39, False=@n37) {group=00_Switches, parameterName=IsBody, sortPriority=0})
  )
  n41 = ScalarParameter() {defaultValue=1, group=Color, parameterName=ShadinSteps}
  n42 = Divide(A=Round(None=Multiply(A=@n40, B=@n41)), B=@n41)
  n43 = ScalarParameter() {defaultValue=0.3016, group=Color, parameterName=MinDarkness}
  n44 = LinearInterpolate(
   A=@n22.RGB
   B=VectorParameter() {group=Color, parameterName=Color, defaultValue=<rgba 0.957 0.2492 0.6909 1>}.RGB
   Alpha=Saturate(None=Max(A=@n42, B=@n43))
  )
  n45 = Multiply(
   A=MaterialFunctionCall(
    Color_Morning=VectorParameter() {group=TODM, parameterName=ColorTint_Morning, defaultValue=<rgba 1 0.6979 0.7786 1>}.RGB
    Color_Day=VectorParameter() {group=TODM, parameterName=ColorTint_Day, defaultValue=<rgba 0.8333 0.8333 0.8333 1>}.RGB
    Color_Evening=VectorParameter() {group=TODM, parameterName=ColorTint_Evening, defaultValue=<rgba 0.7326 0.483 0.8667 1>}.RGB
    Color_Night=VectorParameter() {group=TODM, parameterName=ColorTint_Night, defaultValue=<rgba 0.415 0.63 0.7812 1>}.RGB
   ) {materialFunction=&Color_Time_of_Day}.Result
   B=Multiply(A=ScalarParameter() {defaultValue=0.5849, group=Color, parameterName=Brightness}, B=@n44)
  )
  n46 = VectorParameter() {group=Color, parameterName=InteriorColor, defaultValue=<rgba 0.2247 0.0014 0.25 1>}
  n47 = NamedRerouteUsage()
  n48 = Multiply(A=@n47, B=NamedRerouteUsage())
  n49 = ScalarParameter() {defaultValue=1, group=InteriorGLow, parameterName=FresnelMaskAMount}
  n50 = LinearInterpolate(A=@n45, B=@n46.RGB, Alpha=Saturate(None=LinearInterpolate(A=@n47, B=@n48, Alpha=@n49)))
  n51 = Multiply(A=LinearInterpolate(A=ScalarParameter() {defaultValue=1, group=Color, parameterName=BrightnessFace}, Alpha=VertexColor().B), B=@n50)
  n52 = ScalarParameter() {defaultValue=0.75, group=Color, parameterName=OutlineAmount}
  n53 = LinearInterpolate(Alpha=NamedRerouteUsage()) {constA=-1}
  n54 = StaticSwitchParameter(True=@n50, False=LinearInterpolate(A=LinearInterpolate(A=@n21, B=@n51, Alpha=@n52) {constAlpha=0.65}, B=@n51, Alpha=Saturate(None=@n53))) {group=00_Switches, parameterName=IsBody, sortPriority=0}
  n55 = Power(Base=@n54, Exp=ScalarParameter() {defaultValue=3, group=Color, parameterName=GlowContrast})
  n56 = ScalarParameter() {defaultValue=0.1, group=Color, parameterName=Glow}
  n57 = Multiply(A=Add(A=Multiply(A=@n55, B=@n56), B=@n54), B=NamedRerouteUsage())
  EyeAdaptationInverse(LightValueInput=@n57)
  n58 = VertexNormalWS()
  n59 = Multiply(
   A=MaterialFunctionCall() {materialFunction=&ObjectScale}.Scale Z
   B=ScalarParameter() {defaultValue=16.1339, group=Normal, parameterName=OffsetCardNormal}
  )
  n60 = Multiply(A=VertexColor() {desc="Red is Random\r\nGreen Is Thickness\r\nBlue is Face Mask"}.G, B=@n58)
  n61 = Subtract(A=WorldPosition().XYZ, B=Subtract(A=WorldPosition() {worldPositionShaderOffset=WPT_ExcludeAllShaderOffsets}.XYZ, B=Multiply(A=@n59, B=@n60)))
  LinearInterpolate(A=@n58, B=@n61, Alpha=VertexColor() {desc="Red is Random\r\nGreen Is Thickness\r\nBlue is Face Mask"}.B)
  n62 = VertexNormalWS()
  n63 = Constant3Vector()
  n64 = WorldPosition()
  n65 = Multiply(
   A=VertexNormalWS()
   B=Multiply(
    A=ScalarParameter() {defaultValue=-1, group=WPO, parameterName=BodyInset}
    B=MaterialFunctionCall() {materialFunction=&ObjectScale}.Scale Z
   )
  )
  n66 = WorldPosition()
  n67 = VertexColor()
  n68 = Add(
   A=LinearInterpolate(
    A=ScalarParameter() {defaultValue=6.202, group=WPO, parameterName=Sprite_Scale_Thin}
    B=ScalarParameter() {defaultValue=14.0368, group=WPO, parameterName=Sprite_Scale_Sides}
    Alpha=NamedRerouteUsage()
   )
   B=Multiply(A=ScalarParameter() {defaultValue=3.9214, group=WPO, parameterName=Sprite_Scale_Random}, B=Subtract(A=@n67.R) {constB=0.5})
  )
  n69 = LinearInterpolate(
   A=ScalarParameter() {defaultValue=-0.9246, group=WPO, parameterName=Sprite_Scale_Back_Thin}
   B=ScalarParameter() {defaultValue=29.4715, group=WPO, parameterName=Sprite_Scale_Back}
   Alpha=NamedRerouteUsage()
  )
  n70 = VertexNormalWS()
  n71 = WorldPosition() {worldPositionShaderOffset=WPT_CameraRelativeNoOffsets}
  n72 = NamedRerouteDeclaration(None=Normalize(VectorInput=@n71.XYZ)) {name=CameraVector_Custom, nodeColor=<rgba 0 1 0.8828 1>}
  n73 = Saturate(None=NamedRerouteDeclaration(None=DotProduct(A=@n70, B=@n72)) {name=DotSurface, nodeColor=<rgba 0.2422 0 1 1>})
  n74 = ScalarParameter() {group=WPO, parameterName=Distance_Offset}
  n75 = MaterialFunctionCall() {materialFunction=&ObjectScale}
  n76 = Divide(
   A=Subtract(A=Length(None=@n71.XYZ) {is=class MaterialExpressionLength}, B=Multiply(A=@n74, B=@n75.Scale Z))
   B=Multiply(A=ScalarParameter() {defaultValue=29.5774, group=WPO, parameterName=Distance_Blend}, B=@n75.Scale Z)
  )
  n77 = Multiply(
   A=LinearInterpolate(
    A=ScalarParameter() {defaultValue=4, group=WPO, parameterName=ScaleNear}
    B=LinearInterpolate(A=@n68, B=@n69, Alpha=@n73)
    Alpha=Saturate(None=@n76)
   )
   B=MaterialFunctionCall() {materialFunction=&ObjectScale}.Scale Z
  )
  n78 = LinearInterpolate(
   A=ScalarParameter() {defaultValue=-2.2442, group=WPO, parameterName=SurfaceOffset_thin}
   B=ScalarParameter() {defaultValue=-4.9573, group=WPO, parameterName=SurfaceOffset}
   Alpha=NamedRerouteUsage()
  )
  n79 = OneMinus(None=Saturate(None=NamedRerouteUsage()))
  n80 = Saturate(None=Multiply(A=NamedRerouteUsage()) {constB=-1})
  n81 = LinearInterpolate(
   A=ScalarParameter() {defaultValue=-4.8221, group=WPO, parameterName=PushBackFront_thin}
   B=ScalarParameter() {defaultValue=-4.3507, group=WPO, parameterName=PushBackFront}
   Alpha=NamedRerouteUsage()
  )
  n82 = Add(A=Multiply(A=VertexNormalWS(), B=Multiply(A=@n78, B=@n79)), B=Multiply(A=NamedRerouteUsage(), B=Multiply(A=@n80, B=@n81)))
  n83 = MaterialFunctionCall() {materialFunction=&ObjectScale}
  n84 = Multiply(
   A=Add(
    A=MaterialFunctionCall!(Center Location=@n66.XYZ, 0-1 UVs, XY scale=@n77, OverridePivot, "Normals (Optional)") {materialFunction=&Sprite}.World position offset
    B=Multiply(A=@n82, B=@n83.Scale Z)
   )
   B=ScalarParameter() {defaultValue=1, group=WPO, parameterName=OverallScale}
  )
  LinearInterpolate(
   A=Multiply(A=Multiply(A=@n62) {constB=0}, B=Subtract(A=@n63, B=@n64.XYZ))
   B=Add(A=StaticSwitchParameter(True=@n65, False=@n84) {group=00_Switches, parameterName=IsBody, sortPriority=0}, B=NamedRerouteUsage())
   Alpha=NamedRerouteDeclaration(None=CollectionParameter() {collection=/WildEstate/TechArt/MPCs/MPC_WildEstate_TA, group=WildEstate, parameterName=WPOCollapse}) {name=WPOCollapse, nodeColor=<rgba 0.8984 1 0 1>}
  )
  n85 = TextureObject() {samplerType=SAMPLERTYPE_LinearColor, texture=&HighContrastNoise}
  n86 = VectorParameter() {desc=37.719421, group=InteriorGLow, parameterName=CausticWorldScale, defaultValue=<rgba 8500 8500 8500 0>}
  n87 = WorldPosition() {worldPositionShaderOffset=WPT_ExcludeAllShaderOffsets}
  n88 = Multiply(A=Time()) {constB=0.05}
  n89 = StaticBool()
  n90 = Constant() {r=1}
  n91 = MaterialFunctionCall!(TextureObject=@n85, TextureSize=@n86.RGB, WorldPosition, Export Float 4=@n89, In, World Space Normal, ProjectionTransitionContrast=@n90) {materialFunction=/Game/Athena/SafeZone/WorldAlignedTexture_ScalePan}
  LinearInterpolate(
   A=MaterialFunctionCall(
    TextureObject=@n85
    TextureSize=Multiply(A=@n86.RGB) {constB=0.88}
    WorldPosition=Add(A=@n87.XYZ, B=@n88)
    Export Float 4=@n89
    ProjectionTransitionContrast=@n90
   ) {materialFunction=&WorldAlignedTexture_ScalePan}.XYZ Texture
   B=@n91.XYZ Texture
  )
  n92 = NamedRerouteUsage()
  LinearInterpolate(A=@n22.RGB, B=@n44, Alpha=@n92)
  MaterialFunctionCall(Emissive=@n57) {materialFunction=&MF_FX_EmissiveAdjustments}
  n93 = ScalarParameter() {defaultValue=1.5, group=Color, parameterName=ShadinOffsetOdd}
  MaterialFunctionCall(
   Value_Morning=@n93
   Value_Day=@n29
   Value_Evening=@n93
   Value_Night=ScalarParameter() {defaultValue=1.85, group=Color, parameterName=ShadinOffsetNight}
  ) {materialFunction=&Scalar_Time_of_Day}
  n94 = Transform(None=Constant3Vector() {constant=<rgba 1 0 0 0>}) {transformSourceType=TRANSFORMSOURCE_Camera}
  MaterialFunctionCall(
   Plane Normal Axis=CameraVectorWS()
   Plane X Axis=@n94
   Plane Center=ObjectPositionWS()
   UV Scale=Multiply(
    A=ScalarParameter() {defaultValue=100, group=Noise, parameterName=Noise_Size, sortPriority=36}
    B=MaterialFunctionCall() {materialFunction=&ObjectScale}.Scale Z
   )
  ) {materialFunction=&VirtualPlaneCoordinates}
  Multiply(A=NamedRerouteUsage(), B=@n54)
  n95 = NamedRerouteUsage()
  n96 = StaticSwitchParameter(True=Constant() {r=1}, False=NamedRerouteUsage()) {group=00_Switches, parameterName=IsBody, sortPriority=0}
  Multiply(A=Multiply(A=Multiply(A=@n95, B=@n96), B=NamedRerouteUsage()), B=NamedRerouteUsage())
  Multiply(A=@n45, B=@n46.RGB)
  n97 = Constant2Vector() {g=0.5, r=0.25}
  n98 = PreSkinnedPosition()
  n99 = PreSkinnedLocalBounds()
  n100 = Divide(A=Subtract(A=VertexInterpolator(VS=@n98).PS, B=@n99.Min), B=Subtract(A=@n99.Max, B=@n99.Max))
  n101 = ComponentMask(None=@n100) {g=true, r=true}
  n102 = Multiply(A=@n101, B=@n97)
  n103 = Multiply(A=@n97, B=Time())
  n104 = Multiply(A=Sine(None=Add(A=@n102, B=@n103)), B=Constant2Vector() {desc=Amplitude, g=0.0666, r=0.0666})
  n105 = ComponentMask(None=@n104) {g=true}
  n106 = ComponentMask(None=@n104) {r=true}
  n107 = NamedRerouteDeclaration(None=Add(A=@n101, B=AppendVector(A=@n105, B=@n106))) {name=WavyUVs, nodeColor=<rgba 1 0 0.4687 1>}
  n108 = ConstantBiasScale(None=Add(A=Multiply(A=@n101) {constB=0.5}) {constB=0.25}) {bias=-0.5, scale=2}
  n109 = ComponentMask(None=@n108) {g=true}
  n110 = ComponentMask(None=@n108) {r=true}
  n111 = Arctangent2Fast(Y=@n109, X=@n110)
  n112 = Constant2Vector() {g=113, r=713}
  n113 = Divide(A=@n112.R, B=@n112.R) {constA=710, constB=113}
  n114 = Time()
  n115 = Add(
   A=Frac(None=Divide(A=@n111, B=@n113)) {desc=Angle}
   B=Multiply(A=@n114, B=ScalarParameter() {defaultValue=0.01, group=Distortion, parameterName=Y Speed}) {constB=0.215}
  )
  n116 = Add(
   A=NamedRerouteDeclaration(None=Multiply(B=@n115) {constA=2}) {name=Angle, nodeColor=<rgba 0 1 0.1094 1>}
   B=LinearInterpolate(
    A=ScalarParameter() {defaultValue=0.5, group=Distortion, parameterName=Twist Center}
    B=ScalarParameter() {defaultValue=0.1, group=Distortion, parameterName=Twist Edges}
    Alpha=NamedRerouteUsage()
   )
  )
  n117 = NamedRerouteUsage()
  n118 = TextureSample!(
   UVs=Add(A=@n107, B=AppendVector(A=@n116, B=@n117))
   Tex=NamedRerouteDeclaration(None=TextureObject() {samplerType=SAMPLERTYPE_Grayscale, texture=/Game/Global/Textures/Default/Noise/LowResBlurredNoiseNormalized}) {name=NoiseTex, nodeColor=<rgba 1 0.9844 0 1>}
   Level=QualitySwitch(Default=Constant(), Low=Constant() {r=4}, Medium=Constant() {r=1})
   Apply View MipBias
  ) {mipValueMode=TMVM_MipLevel, samplerType=SAMPLERTYPE_Grayscale, texture=&LowResBlurredNoiseNormalized}
  n119 = LinearInterpolate(B=@n118.R, Alpha=NamedRerouteUsage())
  n120 = Constant() {r=9}
  Multiply(A=Round(None=Multiply(A=@n119, B=@n120) {constB=5})) {constB=0.2}
  Multiply() {constB=0.5}
  Multiply() {constB=0.5}
  Multiply()
  n121 = CameraVectorWS()
  n122 = Normalize(VectorInput=Transform(None=Constant3Vector() {constant=<rgba 1 0 0 0>}) {transformSourceType=TRANSFORMSOURCE_Camera})
  n123 = ObjectPositionWS()
  n124 = Multiply(
   A=ScalarParameter() {defaultValue=100, group=Noise, parameterName=Noise_Size, sortPriority=36}
   B=MaterialFunctionCall() {materialFunction=&ObjectScale}.Scale Z
  )
  n125 = Multiply(B=@n124) {constA=0.5}
  NamedRerouteDeclaration(None=TextureSample!(
    UVs=MaterialFunctionCall(Plane Normal Axis=@n121, Plane X Axis=@n122, Plane Center=@n123, UV Scale=@n125) {materialFunction=&VirtualPlaneCoordinates}.UVs
    Tex
    Apply View MipBias
   ) {samplerType=SAMPLERTYPE_Masks, texture=/WildEstate/Assets/FX/GasCloud/GasCloud_Tiling_M}.R) {name=NoiseTexture, nodeColor=<rgba 0.4297 1 0 1>}
  n126 = Distance(A=WorldPosition().XYZ, B=CameraPositionWS())
  n127 = ScalarParameter() {defaultValue=1594.8198, group=CameraFade, parameterName=CameraFade_Offset}
  n128 = Saturate(None=Divide(A=Subtract(A=@n126, B=@n127), B=ScalarParameter() {defaultValue=953.3473, group=CameraFade, parameterName=CameraFade_Distance}))
  n129 = NamedRerouteUsage()
  n130 = MaterialFunctionCall(Transition Phase=@n128, Height Texture=ComponentMask(None=@n129) {r=true}) {materialFunction=&HeightLerp}
  n131 = Subtract(A=@n130.Alpha, B=ScalarParameter() {defaultValue=-0.6584, group=CameraFade, parameterName=IntersectionOutline})
  n132 = ComponentMask(None=@n129) {r=true}
  n133 = Floor(None=@n128)
  NamedRerouteDeclaration(None=Multiply(A=Floor(None=@n131), B=LinearInterpolate(A=@n132, Alpha=@n133) {constA=1})) {name=OutlinesIntersection, nodeColor=<rgba 0.0547 0 1 1>}
  n134 = SquareRoot(None=Add(A=Multiply(A=@n109, B=@n109), B=Multiply(A=@n110, B=@n110))) {desc=Radius}
  n135 = Multiply(A=@n114, B=ScalarParameter() {defaultValue=-0.013, group=Distortion, parameterName=X Speed}) {constB=0.05}
  NamedRerouteDeclaration(None=Multiply(A=Add(A=@n134, B=@n135)) {constB=4}) {name=Radius, nodeColor=<rgba 0.4297 0 1 1>}
  n136 = NamedRerouteDeclaration(None=MaterialFunctionCall(Plane Normal Axis=@n121, Plane X Axis=@n122, Plane Center=@n123, UV Scale=@n124) {materialFunction=&VirtualPlaneCoordinates}.UVs) {name=VPUV, nodeColor=<rgba 0.6875 1 0 1>}
  NamedRerouteDeclaration(None=TextureSample(UVs=@n136) {samplerType=SAMPLERTYPE_Grayscale, texture=&LowResBlurredNoise}.R) {name=NoiseTexture_Gray, nodeColor=<rgba 0.4297 1 0 1>}
  n137 = MaterialFunctionCall(
   Color_Morning=VectorParameter() {group=Color, parameterName=GlowUnder_Morning, defaultValue=<rgba 0.2292 0.1723 0.2007 1>}.RGB
   Color_Day=VectorParameter() {group=Color, parameterName=GlowUnder_Day, defaultValue=<rgba 0.0402 0.0344 0.125 1>}.RGB
   Color_Evening=VectorParameter() {group=Color, parameterName=GlowUnder_Evening, defaultValue=<rgba 0.0781 0.0587 0.0684 1>}.RGB
   Color_Night=VectorParameter() {group=Color, parameterName=GlowUnder_Night, defaultValue=<rgba 0.0783 0.0264 0.2812 1>}.RGB
  ) {materialFunction=&Color_Time_of_Day}
  n138 = Fresnel(ExponentIn=ScalarParameter() {defaultValue=2.3911, group=Color, parameterName=Fresnel}) {baseReflectFraction=0}
  n139 = MaterialFunctionCall(
   Color_Morning=VectorParameter() {group=Color, parameterName=Glow_Morning, defaultValue=<rgba 0.3802 0.097 0.2482 1>}.RGB
   Color_Day=VectorParameter() {group=Color, parameterName=Glow_Day, defaultValue=<rgba 0.106 0.0555 0.1146 1>}.RGB
   Color_Evening=VectorParameter() {group=Color, parameterName=Glow_Evening, defaultValue=<rgba 0.1406 0.0798 0.1123 1>}.RGB
   Color_Night=VectorParameter() {group=Color, parameterName=Glow_Night, defaultValue=<rgba 0.1561 0.0248 0.1562 1>}.RGB
  ) {materialFunction=&Color_Time_of_Day}
  n140 = ScalarParameter() {defaultValue=1, group=Color, parameterName=FresnelShadinSteps}
  n141 = Divide(A=Round(None=Multiply(A=@n138, B=@n140)), B=@n140)
  n142 = Power(Base=Add(A=MaterialFunctionCall() {materialFunction=&BoundingBoxBased_0-1_UVW}.B) {constB=0.5}) {constExponent=8}
  NamedRerouteDeclaration(None=LinearInterpolate(A=Multiply(A=@n137.Result, B=@n138), B=Multiply(A=@n139.Result, B=@n141), Alpha=Saturate(None=@n142))) {name=FresnelGlow, nodeColor=<rgba 1 0 0.5391 1>}
  n143 = NamedRerouteUsage()
  n144 = NamedRerouteUsage()
  n145 = MaterialFunctionCall(Transition Phase=@n143, Height Texture=@n144) {materialFunction=&HeightLerp}
  n146 = NamedRerouteUsage()
  n147 = MaterialFunctionCall(Transition Phase=@n143, Height Texture=@n146) {materialFunction=&HeightLerp}
  NamedRerouteDeclaration(None=StaticSwitchParameter(
    True=LinearInterpolate(A=@n145.Alpha, B=@n147.Alpha, Alpha=@n147.Alpha)
    False=MaterialFunctionCall(Transition Phase=@n143, Height Texture=@n146) {materialFunction=&HeightLerp}.Alpha
   ) {group=00_Switches, parameterName=IsBody, sortPriority=0}) {name=DissolveValue, nodeColor=<rgba 0 0.9531 1 1>}
  n148 = PreSkinnedLocalBounds()
  n149 = Subtract(A=VertexInterpolator(VS=PreSkinnedPosition()).PS, B=@n148.Min)
  n150 = Subtract(A=@n148.Max, B=@n148.Max)
  n151 = ScalarParameter() {defaultValue=0.01, group=InteriorGLow, parameterName=InteriorGlow_NoiseAmount}
  n152 = ComponentMask(None=NamedRerouteUsage()) {b=true}
  n153 = ComponentMask(None=Add(A=Divide(A=@n149, B=@n150), B=Multiply(A=@n151, B=@n152) {constA=0.25})) {b=true, r=true}
  n154 = Constant2Vector() {g=0.5, r=0.5}
  n155 = MaterialFunctionCall(Speed=ScalarParameter() {defaultValue=0.2, group=InteriorGLow, parameterName=InteriorGlow_SineSpeed}) {materialFunction=&DebugTimeSine}
  n156 = LinearInterpolate(
   A=ScalarParameter() {defaultValue=93.0634, group=InteriorGLow, parameterName=InteriorGlow_SphereRadius}
   B=ScalarParameter() {defaultValue=0.25, group=InteriorGLow, parameterName=InteriorGlow_SphereRadiusMax}
   Alpha=@n155.Result
  )
  n157 = ScalarParameter() {defaultValue=1, group=InteriorGLow, parameterName=InteriorGlow_SphereHardness}
  NamedRerouteDeclaration(None=Saturate(None=SphereMask(A=@n153, B=@n154, Radius=@n156, Hardness=@n157))) {name=BoundMask, nodeColor=<rgba 0.3594 0 1 1>}
  NamedRerouteDeclaration(None=@n141) {name=FresnelMask, nodeColor=<rgba 1 0.7031 0 1>}
  NamedRerouteDeclaration(None=@n155.Result) {name=TimeSIne, nodeColor=<rgba 0 0.9062 1 1>}
  n158 = Constant()
  n159 = Multiply(
   A=OneMinus(None=TextureSample() {samplerType=SAMPLERTYPE_LinearGrayscale, texture=/WildEstate/Assets/FX/GasCloud/GasCloud_height}.R)
   B=Multiply(
    A=ScalarParameter() {defaultValue=3, group=Pixel Depth Offset, parameterName=PDO}
    B=MaterialFunctionCall() {materialFunction=&ObjectScale}.Scale Z
   )
  )
  NamedRerouteDeclaration(None=Multiply(A=StaticSwitchParameter(True=@n158, False=@n159) {group=00_Switches, parameterName=IsBody, sortPriority=0}, B=NamedRerouteUsage())) {name=PDO, nodeColor=<rgba 1 0 0.2812 1>}
  NamedRerouteDeclaration(None=Power(Base=@n134)) {name=RippleMask, nodeColor=<rgba 0 1 0.3438 1>}
  NamedRerouteDeclaration(None=Saturate(None=@n119)) {name=RippleValue, nodeColor=<rgba 1 0 0.2109 1>}
  NamedRerouteDeclaration(None=Saturate(None=@n11)) {name=CameraAngleMask, nodeColor=<rgba 0 1 0.2969 1>}
  NamedRerouteDeclaration(None=@n91.XYZ Texture) {name=DefaultTriPlanarNoise, nodeColor=<rgba 1 0 0.0703 1>}
  NamedRerouteDeclaration(None=TextureSample(UVs=@n107) {samplerType=SAMPLERTYPE_Masks, texture=&GasCloud_Tiling_M}.RGB) {name=Noisetexture blend, nodeColor=<rgba 0 1 0.0391 1>}
  n160 = Divide(B=MaterialFunctionCall() {materialFunction=&ObjectScale}.Scale Z)
  n161 = Constant4Vector() {constant=<rgba 0 0 5 -1>}
  n162 = ComponentMask(None=@n161.RGBA) {b=true, g=true, r=true}
  n163 = MaterialFunctionCall!(Center Point World Position=LinearInterpolate(
    A=TransformPosition(None=@n162) {transformType=TRANSFORMPOSSOURCE_World}
    B=ComponentMask(None=@n161.RGBA) {b=true, g=true, r=true}
    Alpha=Constant()
   )) {materialFunction=/Engine/Functions/Engine_MaterialFunctions02/UVs/WorldSpaceAlignedScreenCoordinates}
  n164 = Constant() {r=-1000}
  n165 = PreSkinnedLocalBounds()
  n166 = MaterialFunctionCall(Float3=Subtract(A=@n165.Max, B=@n165.Max)) {materialFunction=&BreakOutFloat3Components}
  n167 = Multiply(A=ComponentMask(None=@n161.RGBA) {a=true}, B=Max(A=Max(A=@n166.R, B=@n166.R), B=@n166.B))
  n168 = Add(A=Multiply(A=Multiply(A=@n160, B=@n163.Offset and Scaled UVS), B=Divide(A=@n164, B=@n167))) {constB=0.5}
  n169 = ConstantBiasScale(None=@n168) {bias=-0.5, scale=2}
  n170 = ComponentMask(None=@n169) {g=true}
  n171 = ComponentMask(None=@n169) {r=true}
  n172 = SquareRoot(None=Add(A=Multiply(A=@n170, B=@n170), B=Multiply(A=@n171, B=@n171))) {desc=Radius}
  n173 = Time()
  n174 = Multiply(A=@n173, B=ScalarParameter() {defaultValue=-0.013, group=Distortion, parameterName=X Speed}) {constB=0.05}
  NamedRerouteDeclaration(None=Multiply(A=Add(A=@n172, B=@n174)) {constB=4}) {name=Radius_1, nodeColor=<rgba 0.4297 0 1 1>}
  NamedRerouteDeclaration(None=Power(Base=VertexColor().G, Exp=ScalarParameter() {defaultValue=0.6469, group=WPO, parameterName=Thickness_Falloff})) {name=ThicknessMask, nodeColor=<rgba 1 0.2812 0 1>}
  NamedRerouteDeclaration(None=Power(Base=@n172)) {name=RippleMask_1, nodeColor=<rgba 0 1 0.3438 1>}
  n175 = Constant2Vector() {g=0.5, r=0.25}
  n176 = Multiply(A=@n168, B=@n175)
  n177 = Multiply(A=@n175, B=Time())
  n178 = Multiply(A=Sine(None=Add(A=@n176, B=@n177)), B=Constant2Vector() {desc=Amplitude, g=0.0666, r=0.0666})
  n179 = Add(A=@n168, B=AppendVector(A=ComponentMask(None=@n178) {g=true}, B=ComponentMask(None=@n178) {r=true}))
  n180 = Arctangent2Fast(Y=@n170, X=@n171)
  n181 = Constant2Vector() {g=113, r=713}
  n182 = Divide(A=@n181.R, B=@n181.R) {constA=710, constB=113}
  n183 = Add(
   A=Frac(None=Divide(A=@n180, B=@n182)) {desc=Angle}
   B=Multiply(A=@n173, B=ScalarParameter() {defaultValue=0.01, group=Distortion, parameterName=Y Speed}) {constB=0.215}
  )
  n184 = Add(
   A=NamedRerouteDeclaration(None=Multiply(B=@n183) {constA=2}) {name=Angle_1, nodeColor=<rgba 0 1 0.1094 1>}
   B=LinearInterpolate(
    A=ScalarParameter() {defaultValue=0.5, group=Distortion, parameterName=Twist Center}
    B=ScalarParameter() {defaultValue=0.1, group=Distortion, parameterName=Twist Edges}
    Alpha=NamedRerouteUsage()
   )
  )
  n185 = NamedRerouteUsage()
  n186 = TextureSample!(
   UVs=Add(A=@n179, B=AppendVector(A=@n184, B=@n185))
   Tex=NamedRerouteUsage()
   Level=QualitySwitch(Default=Constant(), Low=Constant() {r=4}, Medium=Constant() {r=1})
   Apply View MipBias
  ) {mipValueMode=TMVM_MipLevel, samplerType=SAMPLERTYPE_Grayscale, texture=&LowResBlurredNoiseNormalized}
  n187 = NamedRerouteUsage()
  NamedRerouteDeclaration(None=Saturate(None=LinearInterpolate(B=@n186.R, Alpha=@n187))) {name=RippleValue_1, nodeColor=<rgba 1 0 0.2109 1>}
  NamedRerouteDeclaration(None=@n100) {name=CalculatedBounds, nodeColor=<rgba 1 0 0.0234 1>}
  NamedRerouteDeclaration(None=TextureSample!(UVs=Multiply(B=@n179) {constA=3}, Tex, Apply View MipBias) {samplerType=SAMPLERTYPE_Masks, texture=&GasCloud_Tiling_M}.RGB) {name=Noisetexture blend_1, nodeColor=<rgba 0 1 0.0391 1>}
  n188 = Normalize(VectorInput=@n61)
  n189 = ScalarParameter() {group=Normal, parameterName=CardUseVertexNormal}
  NamedRerouteDeclaration(None=StaticSwitchParameter(True=@n58, False=LinearInterpolate(A=@n188, B=@n58, Alpha=@n189)) {group=00_Switches, parameterName=IsBody, sortPriority=0}) {name=Normals, nodeColor=<rgba 1 0.7031 0 1>}
  n190 = VertexColor()
  n191 = Subtract(A=@n190.R) {constB=0.5}
  n192 = Multiply(A=Time(), B=Multiply(A=ScalarParameter() {defaultValue=1, group=Animation, parameterName=Rotation_Speed}, B=Sign(None=@n191)))
  n193 = Multiply(A=@n190.R, B=ScalarParameter() {defaultValue=44.0961, group=Animation, parameterName=RandomUVRot})
  n194 = TextureSample(UVs=Rotator(Time=Add(A=@n192, B=@n193))) {samplerType=SAMPLERTYPE_Masks, texture=/WildEstate/Assets/FX/GasCloud/GasCloud_M}
  NamedRerouteDeclaration(None=@n194.G) {name=TextureOutline, nodeColor=<rgba 0.3828 0 1 1>}
  NamedRerouteDeclaration(None=@n194.R) {name=TextureMask, nodeColor=<rgba 0 1 0.0625 1>}
  n195 = ScalarParameter() {defaultValue=0.3, group=Animation, parameterName=SineMovement_Amp}
  n196 = MaterialFunctionCall() {materialFunction=&ObjectScale}
  n197 = ScalarParameter() {defaultValue=11.1265, group=Animation, parameterName=SinSIze}
  n198 = Add(
   A=Divide(A=WorldPosition() {worldPositionShaderOffset=WPT_ExcludeAllShaderOffsets}.XYZ, B=Multiply(A=@n197, B=@n196.Scale Z))
   B=Multiply(A=ScalarParameter() {defaultValue=0.5, group=Animation, parameterName=SineMovement_Speed}, B=Time())
  )
  NamedRerouteDeclaration(None=Multiply(A=Multiply(A=@n195, B=@n196.Scale Z), B=Sine(None=@n198))) {name=SineWave Wobble, nodeColor=<rgba 0.5 1 0 1>}
  NamedRerouteDeclaration(None=@n130.Alpha) {name=NearCamera Mask, nodeColor=<rgba 0.0547 0 1 1>}
  n199 = ComponentMask(None=NamedRerouteUsage()) {g=true, r=true}
  NamedRerouteDeclaration(None=Multiply(A=ScalarParameter() {defaultValue=0.5, group=Noise, parameterName=Noise_Add}, B=@n199)) {name=Noise, nodeColor=<rgba 0.2656 1 0 1>}
  NamedRerouteUsage()
  NamedRerouteUsage()
  NamedRerouteUsage()
  NamedRerouteUsage()
  Normalize(VectorInput=@n94)
  OneMinus(None=@n118.R)
  OneMinus(None=@n186.R)
  n200 = Constant() {r=-0.5}
  n201 = Constant() {r=0.5}
  Saturate(None=LinearInterpolate(Alpha=MaterialFunctionCall(Input=@n144, Input Low=@n200, Input High=@n201) {materialFunction=&RemapValue_0-1}.Result) {constA=1, constB=0})
  Saturate(None=@n92)
  Saturate(None=@n26)
  Subtract(A=@n199) {constB=0.5}
  TextureCoordinate()
  TextureObject() {samplerType=SAMPLERTYPE_Normal, texture=/Game/TheOutpost/Textures/TilingWeakPointNormal}
  TextureSample(UVs=@n136) {samplerType=SAMPLERTYPE_Masks, texture=&GasCloud_Tiling_M}
  TextureSample(UVs=@n107) {samplerType=SAMPLERTYPE_Grayscale, texture=&LowResBlurredNoise}
  WorldPosition()
 ]
}
