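# Translucent two-sided unlit glow material for the WildEstate donut cloud, tinting a glow color by time of day and adding a depth-based intersection edge.
# Two panning world-aligned cloud patterns, blended by storm detail and height-lerped, form a GlowNoise mask.
# Opacity combines fresnel, UV gradient, near and depth fades, and a collection-driven WPOCollapse value.
# Low and medium quality add distance fading and an offset dropping far vertices out of view.
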Material /WildEstate/TechArt/DonutCloud/Materials/M_DonutCloud_Glow
graph {
 properties {bAutomaticallySetUsageInEditor=false, blendMode=BLEND_Translucent, shadingModel=MSM_Unlit, twoSided=true, usedShadingModels=Unlit}
 nodes [
  n1 = Multiply(
   A=MaterialFunctionCall(
    Color_Morning=VectorParameter() {group=Color, parameterName=ColorTint_Morning, defaultValue=<rgba 0.75 0.352 0.4583 1>}.RGB
    Color_Day=VectorParameter() {group=Color, parameterName=ColorTint_Day, defaultValue=<rgba 1 1 1 1>}.RGB
    Color_Evening=VectorParameter() {group=Color, parameterName=ColorTint_Evening, defaultValue=<rgba 0.85 0.4035 0.3047 1>}.RGB
    Color_Night=VectorParameter() {group=Color, parameterName=ColorTint_Night, defaultValue=<rgba 0.3034 0.7123 1 1>}.RGB
   ) {materialFunction=&Color_Time_of_Day}.Result
   B=VectorParameter() {group=Color, parameterName=Glow, defaultValue=<rgba 0.3324 0.002 0.3698 1>}.RGB
  )
  n2 = Power(Base=NamedRerouteUsage()) {constExponent=2.25}
  n3 = Power(
   Base=MaterialFunctionCall!(
    Distance Bias=Constant()
    Distance Divisor=ScalarParameter() {defaultValue=250, parameterName=Intersection Edge Divisor}
    Scene Depth
    Pixel Depth
   ) {materialFunction=/Game/Effects/Fort_Effects/Materials/_Functions/MF_DistanceBehindTranslucency_01}.Normalized Inverted Unclamped
   Exp=ScalarParameter() {defaultValue=1, parameterName=Intersection Edge Exponent}
  )
  Add(
   A=@n1
   B=Multiply(A=Multiply(A=@n2, B=@n3), B=VectorParameter() {parameterName=Intersection Edge Color, defaultValue=<rgba 35.9528 0.2213 40 1>}.RGB)
  ) {desc=+ intersection edge}
  Constant3Vector() {constant=<rgba 10000 10000 20000 0>}
  Constant() {r=19000}
  Constant() {r=1.25}
  Distance(A=ObjectPositionWS(), B=CameraPositionWS())
  EyeAdaptationInverse(LightValueInput=Multiply(A=@n1, B=NamedRerouteUsage()))
  n4 = ScalarParameter() {defaultValue=0.4182, group=Opacity, parameterName=Opacity}
  n5 = Divide(A=@n4) {constB=1.33}
  n6 = NamedRerouteUsage()
  n7 = Multiply(A=VertexNormalWS(), B=TwoSidedSign())
  n8 = CameraVectorWS()
  n9 = Saturate(None=Power(Base=DotProduct(A=@n7, B=@n8), Exp=ScalarParameter() {defaultValue=4.7, group=Opacity, parameterName=FresnelEXP}))
  n10 = OneMinus(None=ComponentMask(None=TextureCoordinate()) {g=true})
  n11 = ScalarParameter() {defaultValue=32, group=Opacity, parameterName=Gradient_TopMask, sortPriority=51}
  n12 = Subtract(A=@n10, B=Saturate(None=Power(Base=@n10, Exp=@n11) {constExponent=32}))
  n13 = ScalarParameter() {defaultValue=1, group=Opacity, parameterName=Gradient_Intensity, sortPriority=51}
  n14 = Saturate(None=Power(Base=Multiply(A=@n12, B=@n13), Exp=ScalarParameter() {defaultValue=1, group=Opacity, parameterName=Gradient_Contrast, sortPriority=51}) {constExponent=4})
  n15 = LinearInterpolate(
   A=@n9
   B=Multiply(A=@n9, B=NamedRerouteDeclaration(None=@n14) {name=HeightGrad, nodeColor=<rgba 0.4297 0 1 1>})
   Alpha=ScalarParameter() {group=Opacity, parameterName=Gradient_BlendAmount, sortPriority=50}
  )
  n16 = NamedRerouteUsage()
  n17 = LinearInterpolate(A=Multiply(A=LinearInterpolate(A=@n4, B=@n5, Alpha=@n6), B=Multiply(A=@n15, B=@n16)), Alpha=Saturate(None=@n3)) {constB=0}
  Multiply(A=@n17, B=NamedRerouteUsage())
  n18 = ScalarParameter() {defaultValue=0.3, desc=0.0666, parameterName=Sine Speed A_22}
  n19 = Time() {desc=15.015014, period=1}
  n20 = MaterialFunctionCall!(
   TextureObject=TextureObject() {samplerType=SAMPLERTYPE_LinearColor, texture=/Game/Packages/Fortress_Sky/SkyDome/Master/T_SkyDome_CloudShapes}
   TextureSize=VectorParameter() {parameterName=CausticWorldScale, defaultValue=<rgba 10000 10000 20000 0>}.RGB
   WorldPosition
   Export Float 4=StaticBool()
   In
   World Space Normal
   ProjectionTransitionContrast=Constant() {r=10}
  ) {materialFunction=/Game/Athena/SafeZone/WorldAlignedTexture_ScalePan}
  n21 = Frac(None=Add(A=Multiply(A=@n18, B=@n19), B=@n20.XY Texture))
  n22 = ScalarParameter() {defaultValue=-0.1, desc=0.0666, parameterName=Sine Speed A_1}
  n23 = Frac(None=Add(A=@n20.XY Texture, B=Multiply(A=@n19, B=@n22)))
  n24 = ScalarParameter() {defaultValue=-0.25, parameterName=Param_zzz}
  n25 = ScalarParameter() {defaultValue=1.5, parameterName=vswr23}
  n26 = TextureCoordinate()
  n27 = MaterialFunctionCall(X=ScalarParameter() {defaultValue=9.9571, parameterName=ge5yt3}, Y=ScalarParameter() {defaultValue=8, parameterName=Param_2vd23}) {materialFunction=&MakeFloat2}
  n28 = Constant2Vector() {g=1}
  n29 = Multiply(A=Time(), B=ScalarParameter() {defaultValue=-0.01, parameterName=VertPanSpeedbe4t322})
  n30 = TextureSample(UVs=Add(A=Multiply(A=@n26, B=@n27.Result), B=Multiply(A=@n28, B=@n29))) {samplerType=SAMPLERTYPE_LinearGrayscale, texture=/Game/Global/Textures/Default/Tileable_Patterns/T_StormCloudDetail01_M}
  n31 = LinearInterpolate(
   A=Distance(A=ComponentMask(None=@n21) {g=true}, B=Constant() {r=0.5})
   B=Distance(A=ComponentMask(None=@n23) {g=true}, B=Constant() {r=0.5})
   Alpha=Saturate(None=LinearInterpolate(A=@n24, B=@n25, Alpha=@n30.R))
  )
  n32 = Constant() {r=5}
  Multiply(A=Round(None=Multiply(A=@n31, B=@n32) {constB=5})) {constB=0.2}
  n33 = NamedRerouteUsage()
  n34 = Add(A=PixelDepth().R, B=Constant() {r=-100})
  n35 = ScalarParameter() {defaultValue=300, parameterName=NearFade}
  n36 = MaterialFunctionCall(Vector 2=ComponentMask(None=WorldPosition() {worldPositionShaderOffset=WPT_CameraRelative}.XYZ) {g=true, r=true}) {materialFunction=/Engine/Functions/Engine_MaterialFunctions02/Utility/VectorLength}
  n37 = Divide(A=Add(A=Constant() {r=-100}, B=@n36.V2 Length), B=ScalarParameter() {defaultValue=44900, parameterName=FarFade})
  n38 = Multiply(A=@n17, B=OneMinus(None=Saturate(None=@n37)))
  n39 = Multiply(A=Saturate(None=Divide(A=@n34, B=@n35)), B=QualitySwitch(Default=@n17, Low=@n38, Medium=@n38))
  Multiply(
   A=DepthFade(Opacity=Multiply(A=@n33, B=@n39), FadeDistance=ScalarParameter() {defaultValue=150, group=Opacity, parameterName=FadeDistance}) {fadeDistanceDefault=256}
   B=NamedRerouteDeclaration(None=CollectionParameter() {collection=/WildEstate/TechArt/MPCs/MPC_WildEstate_TA, group=WildEstate, parameterName=WPOCollapse}) {name=WPOCollapse, nodeColor=<rgba 0.8984 1 0 1>}
  )
  n40 = Divide(A=WorldPosition().XYZ, B=ScalarParameter() {defaultValue=7000, parameterName=HeightGradientScale})
  n41 = ScalarParameter() {defaultValue=0.38, parameterName=HeightGradientHeight}
  n42 = Saturate(None=ComponentMask(None=Add(A=@n40, B=@n41)) {b=true})
  n43 = Multiply(A=@n42, B=ScalarParameter() {defaultValue=0.15, parameterName=zzzdwe})
  n44 = MaterialFunctionCall(Transition Phase=@n42, Height Texture=Clamp(None=@n31) {maxDefault=0.75, minDefault=0.25}) {materialFunction=&HeightLerp}
  NamedRerouteDeclaration(None=Saturate(None=Add(A=@n43, B=@n44.Lerp Alpha No Contrast))) {name=GlowNoise, nodeColor=<rgba 0.8047 0 1 1>}
  n45 = WorldPosition()
  n46 = CameraPositionWS()
  n47 = Divide(
   A=Subtract(A=Distance(A=@n45.XYZ, B=@n46), B=ScalarParameter() {defaultValue=800, group=CameraFade, parameterName=CameraFade_Offset})
   B=ScalarParameter() {defaultValue=500, group=CameraFade, parameterName=CameraFade_Distance}
  )
  n48 = NamedRerouteUsage()
  NamedRerouteDeclaration(None=MaterialFunctionCall(Transition Phase=Saturate(None=@n47), Height Texture=ComponentMask(None=@n48) {r=true}) {materialFunction=&HeightLerp}.Alpha) {name=NearCamera Mask, nodeColor=<rgba 0.0547 0 1 1>}
  n49 = CameraVectorWS()
  n50 = Normalize(VectorInput=Transform(None=Constant3Vector() {constant=<rgba 1 0 0 0>}) {transformSourceType=TRANSFORMSOURCE_Camera})
  n51 = ObjectPositionWS()
  n52 = Multiply(
   A=ScalarParameter() {defaultValue=200, group=Distortion, parameterName=Noise_Size, sortPriority=36}
   B=MaterialFunctionCall() {materialFunction=&ObjectScale}.Scale Z
  )
  NamedRerouteDeclaration(None=TextureSample(UVs=MaterialFunctionCall(Plane Normal Axis=@n49, Plane X Axis=@n50, Plane Center=@n51, UV Scale=@n52) {materialFunction=&VirtualPlaneCoordinates}.UVs) {samplerType=SAMPLERTYPE_LinearGrayscale, texture=/WildEstate/Assets/FX/GasCloud/GasCloud_Tiling_Gray}.R) {name=NoiseTexture_Gray, nodeColor=<rgba 0.4297 1 0 1>}
  n53 = Distance(A=WorldPosition().XYZ, B=CameraPositionWS())
  n54 = ScalarParameter() {defaultValue=16081, parameterName=CameraFadeOffset}
  NamedRerouteDeclaration(None=Divide(A=Subtract(A=@n53, B=@n54), B=ScalarParameter() {defaultValue=40000, parameterName=CameraFadeLength})) {name=SimpleFade, nodeColor=<rgba 1 0.6563 0 1>}
  NamedRerouteUsage()
  Normalize(VectorInput=WorldPosition() {worldPositionShaderOffset=WPT_CameraRelativeNoOffsets}.XYZ)
  OneMinus(None=@n42)
  n55 = Constant()
  n56 = If(
   A=@n36.V2 Length
   B=Multiply(A=ScalarParameter() {defaultValue=44900, parameterName=FarFade}, B=ScalarParameter() {defaultValue=1.1, parameterName=Buffer}) {constB=1.36, desc=add some buffer}
   "A > B"=Constant3Vector() {constant=<rgba 0 0 -999999 0>}
   "A == B"=@n55
   "A < B"=@n55
  )
  QualitySwitch(Default=Constant(), Low=@n56, Medium=@n56)
 ]
}
