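# Translucent unlit, two-sided material for the donut cloud interior, rendering a glowing, swirling refractive distortion tinted by time of day.
# Time-scrolled polar UVs with a center-to-edge twist sample noise and normal maps for stepped ripples and normals, alongside a sine-warped wave normal for pixel-normal-offset refraction.
# Sphere masks, depth fade and a noise-broken camera-distance mask shape the effect.
# The WPOCollapse collection parameter drives mask radii and vertex offset.
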Material /WildEstate/TechArt/DonutCloud/Materials/M_DonutCloud_InteriorDistortion
graph {
 properties {
  bAutomaticallySetUsageInEditor=false
  bUsedWithInstancedStaticMeshes=true
  bUsedWithMeshDeformer=true
  bUsedWithSkeletalMesh=true
  blendMode=BLEND_Translucent
  refractionMethod=RM_PixelNormalOffset
  shadingModel=MSM_Unlit
  twoSided=true
  usedShadingModels=Unlit
 }
 nodes [
  n1 = VectorParameter() {group=Color, parameterName=GlowUnder_Morning, defaultValue=<rgba 0.0938 0.0705 0.0821 1>}
  n2 = VectorParameter() {group=Color, parameterName=GlowUnder_Day, defaultValue=<rgba 0.0402 0.0344 0.125 1>}
  n3 = VectorParameter() {group=Color, parameterName=GlowUnder_Evening, defaultValue=<rgba 0.0781 0.0587 0.0684 1>}
  n4 = VectorParameter() {group=Color, parameterName=GlowUnder_Night, defaultValue=<rgba 0.0392 0.0132 0.1406 1>}
  n5 = NamedRerouteUsage()
  n6 = LinearInterpolate(
   A=LinearInterpolate(A=VectorParameter() {group=Color, parameterName=Glow, defaultValue=<rgba 0.5 0.0465 0.2951 1>}.RGB, B=@n5)
   B=LinearInterpolate(A=@n5, B=Constant4Vector() {constant=<rgba 0.2346 0.1369 0.5417 1>}.RGB)
   Alpha=NamedRerouteUsage()
  )
  n7 = LinearInterpolate(
   A=Multiply(
    A=MaterialFunctionCall(Color_Morning=@n1.RGB, Color_Day=@n2.RGB, Color_Evening=@n3.RGB, Color_Night=@n4.RGB) {materialFunction=&Color_Time_of_Day}.Result
    B=Multiply(A=@n6) {constB=100}
   )
   B=VectorParameter() {group=Color, parameterName=InteriorColor, defaultValue=<rgba 0.8988 0.0055 1 1>}.RGB
   Alpha=NamedRerouteUsage()
  )
  EyeAdaptationInverse(LightValueInput=Multiply(A=@n7, B=NamedRerouteUsage()))
  n8 = NamedRerouteDeclaration(None=CollectionParameter() {collection=/WildEstate/TechArt/MPCs/MPC_WildEstate_TA, group=WildEstate, parameterName=WPOCollapse}) {name=WPOCollapse, nodeColor=<rgba 0.8984 1 0 1>}
  n9 = Multiply(A=Multiply(A=VertexNormalWS()) {constB=0}, B=Subtract(A=Constant3Vector(), B=WorldPosition().XYZ))
  n10 = NamedRerouteUsage()
  LinearInterpolate(A=Multiply(A=@n8, B=Add(A=@n9, B=@n10)), B=Add(A=VertexNormalWS(), B=@n10), Alpha=@n8)
  n11 = Constant3Vector() {constant=<rgba 0 0 1 0>}
  n12 = NamedRerouteUsage()
  n13 = NamedRerouteUsage()
  LinearInterpolate(A=LinearInterpolate(A=@n11, B=NamedRerouteUsage(), Alpha=Multiply(A=@n12, B=@n13)), B=@n11, Alpha=NamedRerouteUsage())
  Multiply(A=NamedRerouteUsage(), B=@n7)
  Multiply(A=NamedRerouteUsage(), B=Multiply(A=NamedRerouteUsage(), B=NamedRerouteUsage()))
  n14 = TextureCoordinate()
  n15 = Constant() {r=0.5}
  n16 = CollectionParameter() {collection=&MPC_WildEstate_TA, group=WildEstate, parameterName=WPOCollapse}
  n17 = SphereMask(
   A=@n14
   B=@n15
   Radius=LinearInterpolate(B=ScalarParameter() {defaultValue=0.4844, group=SphereMask, parameterName=Sphere_Radius}, Alpha=@n16)
   Hardness=ScalarParameter() {defaultValue=0.5, group=SphereMask, parameterName=Sphere_Hardness}
  ) {attenuationRadius=0.475, hardnessPercent=0.5}
  n18 = Multiply(A=NamedRerouteUsage(), B=Multiply(A=ScalarParameter() {defaultValue=1, group=Opacity, parameterName=Opacity}, B=NamedRerouteUsage()))
  Multiply(
   A=Multiply(A=NamedRerouteDeclaration(None=@n17) {name=SphereMask, nodeColor=<rgba 0 0.7422 1 1>}, B=DepthFade(Opacity=@n18))
   B=NamedRerouteUsage()
  )
  n19 = ConstantBiasScale(None=TextureCoordinate()) {bias=-0.5, scale=2}
  n20 = ComponentMask(None=@n19) {g=true}
  n21 = ComponentMask(None=@n19) {r=true}
  n22 = SquareRoot(None=Add(A=Multiply(A=@n20, B=@n20), B=Multiply(A=@n21, B=@n21))) {desc=Radius}
  n23 = Time()
  n24 = Multiply(A=@n23, B=ScalarParameter() {defaultValue=-0.0322, group=Distortion, parameterName=X Speed}) {constB=0.05}
  NamedRerouteDeclaration(None=Multiply(A=Add(A=@n22, B=@n24)) {constB=4}) {name=Radius, nodeColor=<rgba 0.4297 0 1 1>}
  NamedRerouteDeclaration(None=SphereMask(A=@n14, B=@n15, Radius=LinearInterpolate(Alpha=@n16) {constB=0.3}) {attenuationRadius=0.15, hardnessPercent=0.01}) {name=SphereMaskSMall, nodeColor=<rgba 0.9453 1 0 1>}
  n25 = NamedRerouteUsage()
  n26 = NamedRerouteUsage()
  NamedRerouteDeclaration(None=LinearInterpolate(
    A=TextureSample(UVs=@n25, Tex=@n26) {samplerType=SAMPLERTYPE_Normal, texture=/Game/Weapons/WeaponSkins/Textures/T_TileNoise_03_NRM}.RGB
    B=Constant3Vector() {constant=<rgba 0 0 1 0>}
   )) {name=WaveNormal, nodeColor=<rgba 0.3828 0 1 1>}
  n27 = WorldPosition()
  n28 = CameraPositionWS()
  n29 = Divide(
   A=Subtract(A=Distance(A=@n27.XYZ, B=@n28), B=ScalarParameter() {defaultValue=3500, group=CameraFade, parameterName=CameraFade_Offset})
   B=ScalarParameter() {defaultValue=2000, group=CameraFade, parameterName=CameraFade_Distance}
  )
  n30 = NamedRerouteUsage()
  NamedRerouteDeclaration(None=MaterialFunctionCall(Transition Phase=Saturate(None=@n29), Height Texture=ComponentMask(None=@n30) {r=true}) {materialFunction=&HeightLerp}.Alpha) {name=NearCamera Mask, nodeColor=<rgba 0.0547 0 1 1>}
  NamedRerouteDeclaration(None=Saturate(None=OneMinus(None=@n17))) {name=ReverseSPhereMask, nodeColor=<rgba 1 0 0.375 1>}
  n31 = CameraVectorWS()
  n32 = Normalize(VectorInput=Transform(None=Constant3Vector() {constant=<rgba 1 0 0 0>}) {transformSourceType=TRANSFORMSOURCE_Camera})
  n33 = ObjectPositionWS()
  n34 = Multiply(
   A=ScalarParameter() {defaultValue=360, group=Distortion, parameterName=Noise_Size, sortPriority=36}
   B=MaterialFunctionCall() {materialFunction=&ObjectScale}.Scale Z
  )
  NamedRerouteDeclaration(None=TextureSample(UVs=MaterialFunctionCall(Plane Normal Axis=@n31, Plane X Axis=@n32, Plane Center=@n33, UV Scale=@n34) {materialFunction=&VirtualPlaneCoordinates}.UVs) {samplerType=SAMPLERTYPE_LinearGrayscale, texture=/WildEstate/Assets/FX/GasCloud/GasCloud_Tiling_Gray}.R) {name=NoiseTexture_Gray, nodeColor=<rgba 0.4297 1 0 1>}
  n35 = Constant2Vector() {g=113, r=713}
  n36 = Divide(A=Arctangent2Fast(Y=@n20, X=@n21), B=Divide(A=@n35.R, B=@n35.R) {constA=710, constB=113})
  n37 = ScalarParameter() {defaultValue=0.025, group=Distortion, parameterName=Y Speed}
  n38 = Multiply(B=Add(A=Frac(None=@n36) {desc=Angle}, B=Multiply(A=@n23, B=@n37) {constB=0.215})) {constA=2}
  n39 = ScalarParameter() {defaultValue=0.5, group=Distortion, parameterName=Twist Center}
  n40 = ScalarParameter() {defaultValue=0.1, group=Distortion, parameterName=Twist Edges}
  n41 = NamedRerouteUsage()
  n42 = AppendVector(
   A=Add(A=NamedRerouteDeclaration(None=@n38) {name=Angle, nodeColor=<rgba 0 1 0.1094 1>}, B=LinearInterpolate(A=@n39, B=@n40, Alpha=@n41))
   B=NamedRerouteUsage()
  )
  n43 = QualitySwitch(Default=Constant(), Low=Constant() {r=4}, Medium=Constant() {r=1})
  n44 = TextureSample!(UVs=@n42, Tex=NamedRerouteUsage(), Level=@n43, Apply View MipBias) {mipValueMode=TMVM_MipLevel, samplerType=SAMPLERTYPE_Grayscale, texture=/Game/Global/Textures/Default/Noise/LowResBlurredNoiseNormalized}
  n45 = Multiply(A=LinearInterpolate(B=@n44.R, Alpha=NamedRerouteUsage()), B=Constant() {r=5}) {constB=5}
  NamedRerouteDeclaration(None=Multiply(A=Round(None=@n45)) {constB=0.2}) {name=RippleValue, nodeColor=<rgba 1 0 0.2109 1>}
  n46 = Constant3Vector() {constant=<rgba 0 0 1 0>}
  n47 = LinearInterpolate(
   A=@n46
   B=TextureSample(UVs=@n42, Tex=NamedRerouteUsage(), Level=@n43) {mipValueMode=TMVM_MipLevel, samplerType=SAMPLERTYPE_Normal, texture=&T_TileNoise_03_NRM}.RGB
  ) {constAlpha=0.25}
  n48 = NamedRerouteUsage()
  NamedRerouteDeclaration(None=MaterialFunctionCall!(BaseNormal=LinearInterpolate(A=@n46, B=@n47, Alpha=@n48), AdditionalNormal=NamedRerouteUsage()) {materialFunction=/Engine/Functions/Engine_MaterialFunctions02/Utility/BlendAngleCorrectedNormals_WS}.Result) {name=RippleNormal, nodeColor=<rgba 0 0.5312 1 1>}
  NamedRerouteDeclaration(None=TextureObject() {samplerType=SAMPLERTYPE_Normal, texture=/Game/Global/Textures/Default/Noise/T_SwirlNoise_N}) {name=NormalTex, nodeColor=<rgba 0 0.6953 1 1>}
  NamedRerouteDeclaration(None=TextureObject() {samplerType=SAMPLERTYPE_Grayscale, texture=&LowResBlurredNoiseNormalized}) {name=NoiseTex, nodeColor=<rgba 1 0.9844 0 1>}
  n49 = Multiply(A=TextureCoordinate()) {constB=2}
  n50 = Constant2Vector() {g=0.225, r=0.225}
  n51 = Multiply(A=@n49, B=@n50)
  n52 = Multiply(A=@n50, B=Time())
  n53 = Multiply(A=Sine(None=Add(A=@n51, B=@n52)), B=Constant2Vector() {desc=Amplitude, g=0.0666, r=0.0666})
  n54 = ComponentMask(None=@n53) {g=true}
  n55 = ComponentMask(None=@n53) {r=true}
  NamedRerouteDeclaration(None=Add(A=@n49, B=AppendVector(A=@n54, B=@n55))) {name=WaveUVs, nodeColor=<rgba 0.6641 0 1 1>}
  NamedRerouteDeclaration(None=Power(Base=@n22)) {name=RippleMask, nodeColor=<rgba 0 1 0.3438 1>}
  n56 = Constant() {desc=.5}
  n57 = ScalarParameter() {defaultValue=0.25, desc=1.53, group=Refraction, parameterName=IOR, sliderMax=2, sliderMin=1, sortPriority=1}
  n58 = PixelDepth()
  n59 = OneMinus(None=Saturate(None=Divide(A=@n58.R) {constB=1000}))
  n60 = NamedRerouteUsage()
  NamedRerouteDeclaration(None=Multiply(A=LinearInterpolate(A=@n56, B=@n57, Alpha=@n59), B=LinearInterpolate(Alpha=@n60) {constB=0.05})) {name=Refraction, nodeColor=<rgba 0 1 0.4141 1>}
 ]
}
