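# Additive unlit glow overlay material for the Wild Estate donut cloud, usable on skeletal meshes.
# A time-of-day color function tints the glow, which is compensated for eye adaptation.
# An inverted Fresnel concentrates opacity on camera-facing surfaces, and depth fade softens intersections.
# The WPOCollapse collection parameter blends world position offset from zero to a push away from the camera plus inflation along vertex normals.
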Material /WildEstate/TechArt/DonutCloud/Materials/M_DonutCloud_OverlayGlow
graph {
 properties {
  bAutomaticallySetUsageInEditor=false
  bUsedWithMeshDeformer=true
  bUsedWithSkeletalMesh=true
  blendMode=BLEND_Additive
  shadingModel=MSM_Unlit
  usedShadingModels=Unlit
 }
 nodes [
  n1 = MaterialFunctionCall(
   Color_Morning=VectorParameter() {group=Color, parameterName=ColorTint_Morning, defaultValue=<rgba 0.75 0.352 0.4583 1>}.RGB
   Color_Day=VectorParameter() {group=Color, parameterName=ColorTint_Day, defaultValue=<rgba 1 1 1 1>}.RGB
   Color_Evening=VectorParameter() {group=Color, parameterName=ColorTint_Evening, defaultValue=<rgba 0.85 0.4035 0.3047 1>}.RGB
   Color_Night=VectorParameter() {group=Color, parameterName=ColorTint_Night, defaultValue=<rgba 0.3034 0.7123 1 1>}.RGB
  ) {materialFunction=&Color_Time_of_Day}
  n2 = VectorParameter() {group=Color, parameterName=Glow, defaultValue=<rgba 0.5 0.0465 0.2951 1>}
  EyeAdaptationInverse(LightValueInput=Multiply(A=Multiply(A=@n1.Result, B=@n2.RGB), B=NamedRerouteUsage()))
  n3 = NamedRerouteDeclaration(None=CollectionParameter() {collection=/WildEstate/TechArt/MPCs/MPC_WildEstate_TA, group=WildEstate, parameterName=WPOCollapse}) {name=WPOCollapse, nodeColor=<rgba 0.8984 1 0 1>}
  n4 = Multiply(A=VertexNormalWS()) {constB=0}
  n5 = Subtract(A=Constant3Vector(), B=WorldPosition().XYZ)
  n6 = Normalize(VectorInput=WorldPosition() {worldPositionShaderOffset=WPT_CameraRelativeNoOffsets}.XYZ)
  n7 = ScalarParameter() {defaultValue=12613.3809, group=WPO, parameterName=CameraPushBack}
  n8 = VertexNormalWS()
  n9 = ScalarParameter() {defaultValue=2272.1421, group=WPO, parameterName=VertexOffset}
  LinearInterpolate(A=Multiply(A=@n3, B=Multiply(A=@n4, B=@n5)), B=Add(A=Multiply(A=@n6, B=@n7), B=Multiply(A=@n8, B=@n9)), Alpha=@n3)
  n10 = ScalarParameter() {defaultValue=0.632, group=Opacity, parameterName=Opacity}
  n11 = VertexNormalWS()
  n12 = Power(Base=Multiply(A=DotProduct(A=@n11, B=@n6)) {constB=-1}, Exp=ScalarParameter() {defaultValue=7.0136, group=Opacity, parameterName=FresnelEXP})
  Multiply(
   A=DepthFade(Opacity=Multiply(A=@n10, B=@n12), FadeDistance=ScalarParameter() {defaultValue=512, group=Opacity, parameterName=FadeDistance}) {fadeDistanceDefault=50}
   B=NamedRerouteUsage()
  )
 ]
}
